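# Physics asset for the Impulse Spring CAU_03 female medium FaceAcc mesh, setting up dynamic bodies for its ponytail and bangs.
# Capsules on the head, upper spine and clavicles provide body shapes alongside the hair bones.
# Limited-angle constraints chain the ponytail from the head with branching strands, while bangs chains hang from the head or a box-shaped dyn_bangs_00, which has no constraint to the head.
# The right bangs chain locks one swing axis per joint.
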PhysicsAsset /Game/Characters/Player/Female/Medium/Bodies/F_MED_Impulse_Spring/Skins/CAU_03/Meshes/F_MED_Impulse_Spring_CAU_3_FaceAcc_Physics
bodies {
 clavicle_l [{shapeType=Capsule, length=14.5, radius=7, center=<2.2747 9.1369 -3.2414>, rotation=<rot 0 0 122.6963>}]
 clavicle_r [{shapeType=Capsule, length=14.5, radius=7, center=<-3.5588 -8.2111 2.9887>, rotation=<rot 0 0 122.6963>}]
 dyn_bangs_00 [{shapeType=Box, extentX=1.5, extentY=1.5, extentZ=1.5}]
 dyn_bangs_01 [{shapeType=Capsule, length=4.4084, radius=0.7766, center=<1 0 0>, rotation=<rot 90 0 0>}]
 dyn_bangs_02 [{shapeType=Capsule, length=2.1584, radius=0.7766, center=<1.4055 0 0>, rotation=<rot 90 0 0>}]
 dyn_bangs_03 [{shapeType=Capsule, length=1, radius=0.5266, center=<0.8237 0 0.1311>, rotation=<rot 90 0 0>}]
 dyn_bangs_04 [{shapeType=Capsule, length=4.7679, radius=0.786, center=<1 0 0>, rotation=<rot 90 0 0>}]
 dyn_bangs_05 [{shapeType=Capsule, length=2.5179, radius=0.786, center=<1.6151 0 -0.1027>, rotation=<rot 90 0 0>}]
 dyn_bangs_06 [{shapeType=Capsule, length=2.5179, radius=0.786, center=<1.6151 0 -0.1027>, rotation=<rot 90 0 0>}]
 dyn_bangs_08 [{shapeType=Capsule, length=2.5179, radius=0.786, center=<1.5269 0.3361 -0.1028>, rotation=<rot 90 0 -14.691>}]
 dyn_bangs_09 [{shapeType=Capsule, length=2.5179, radius=0.786, center=<1.6151 0 -0.1027>, rotation=<rot 90 0 0>}]
 dyn_bangs_r_01 [{shapeType=Capsule, length=1.25, radius=1.255, center=<0.5999 0 0.1062>, rotation=<rot -79.9496 0.0007 -0.0007>}]
 dyn_bangs_r_02 [{shapeType=Capsule, length=1.505, radius=0.755, center=<1.188 0 -0.1343>, rotation=<rot 83.5486 -0.0002 -0.0002>}]
 dyn_ponytail_01 [{shapeType=Capsule, length=2.25, radius=3.255, center=<1.5123 0 0.026>, rotation=<rot -89.016 0.0009 -0.0009>}]
 dyn_ponytail_02 [{shapeType=Capsule, length=1.75, radius=3.505, center=<1.8936 0 0>, rotation=<rot -90 180 180>}]
 dyn_ponytail_03 [{shapeType=Capsule, length=2.3917, radius=2, center=<2.9494 0 1>, rotation=<rot 82.9703 0.0001 0.0001>}]
 dyn_ponytail_base [{shapeType=Capsule, length=0.5, radius=0.5}]
 dyn_ponytail_parent [{shapeType=Capsule, length=1.25, radius=1.255}]
 dyn_ponytail_strand_01 [{shapeType=Capsule, length=1.2055, radius=0.5, center=<0.9552 0 0>, rotation=<rot 90 0 0>}]
 dyn_ponytail_strand_02 [{shapeType=Capsule, length=0.6617, radius=0.442, center=<0.5815 -0.0997 0.011>, rotation=<rot 0 -88.6245 -87.4558>}]
 dyn_ponytail_strand_03 [{shapeType=Capsule, length=1.1737, radius=0.5545, center=<0.5755 0 -0.3435>, rotation=<rot -59.1626 -180 179.9999>}]
 dyn_ponytail_strand_04 [{shapeType=Capsule, length=1.1737, radius=0.5545, center=<0.8279 0 -0.1667>, rotation=<rot -78.616 -179.9996 179.9996>}]
 dyn_ponytail_strand_05 [{shapeType=Capsule, length=1.0266, radius=0.4563, center=<0.7696 0 0.0404>, rotation=<rot 87.0005 -179.9999 -179.9998>}]
 head [
  {shapeType=Capsule, length=6.75, radius=5.75, center=<0.6488 3.1259 -0.663>, rotation=<rot -87.9079 0 0>}
  {shapeType=Capsule, length=7, radius=5.5, center=<-3.9295 0.5164 -0.5088>, rotation=<rot -87.9079 0 0>}
 ]
 spine_05 [
  {shapeType=Capsule, length=11.5, radius=9.25, center=<2.6545 1.3525 0.0001>}
  {shapeType=Capsule, length=0.25, radius=15, shapeName=spine_05_capsule, center=<7.361 2.0599 0>, rotation=<rot 90 180 160.0005>}
 ]
}
constraints [
 {bone1Name=dyn_ponytail_parent, bone2Name=head, swing1LimitDegrees=5, swing2LimitDegrees=25, twistLimitDegrees=10}
 {bone1Name=dyn_ponytail_01, bone2Name=dyn_ponytail_parent, swing1LimitDegrees=20, swing2LimitDegrees=25, twistLimitDegrees=5}
 {bone1Name=dyn_ponytail_02, bone2Name=dyn_ponytail_01, swing1LimitDegrees=20, swing2LimitDegrees=20, twistLimitDegrees=10}
 {bone1Name=dyn_ponytail_03, bone2Name=dyn_ponytail_02, swing1LimitDegrees=40, swing2LimitDegrees=40, twistLimitDegrees=10}
 {bone1Name=dyn_ponytail_strand_03, bone2Name=dyn_ponytail_02, swing1LimitDegrees=40, swing2LimitDegrees=40, twistLimitDegrees=10}
 {bone1Name=dyn_ponytail_strand_04, bone2Name=dyn_ponytail_02, swing1LimitDegrees=40, swing2LimitDegrees=40, twistLimitDegrees=10}
 {bone1Name=dyn_ponytail_strand_02, bone2Name=dyn_ponytail_01, swing1LimitDegrees=40, swing2LimitDegrees=40, twistLimitDegrees=10}
 {bone1Name=dyn_ponytail_strand_01, bone2Name=dyn_ponytail_01, swing1LimitDegrees=40, swing2LimitDegrees=40, twistLimitDegrees=10}
 {bone1Name=dyn_ponytail_strand_05, bone2Name=dyn_ponytail_02, swing1LimitDegrees=40, swing2LimitDegrees=40, twistLimitDegrees=10}
 {bone1Name=dyn_bangs_r_01, bone2Name=head, swing1LimitDegrees=12, swing2LimitDegrees=3, swing2Motion=Locked, twistLimitDegrees=5}
 {bone1Name=dyn_bangs_r_02, bone2Name=dyn_bangs_r_01, swing1LimitDegrees=3, swing1Motion=Locked, swing2LimitDegrees=10, twistLimitDegrees=5}
 {bone1Name=dyn_bangs_05, bone2Name=dyn_bangs_04, swing1LimitDegrees=12, swing2LimitDegrees=12, twistLimitDegrees=6}
 {bone1Name=dyn_bangs_04, bone2Name=head, swing1LimitDegrees=20, swing2LimitDegrees=12, twistLimitDegrees=6}
 {bone1Name=dyn_bangs_06, bone2Name=dyn_bangs_05, swing1LimitDegrees=12, swing2LimitDegrees=12, twistLimitDegrees=12}
 {bone1Name=dyn_bangs_08, bone2Name=dyn_bangs_00, swing1LimitDegrees=15, swing2LimitDegrees=15, twistLimitDegrees=5}
 {bone1Name=dyn_bangs_09, bone2Name=dyn_bangs_08, swing1LimitDegrees=15, swing2LimitDegrees=15, twistLimitDegrees=5}
 {bone1Name=dyn_bangs_03, bone2Name=dyn_bangs_02, swing1LimitDegrees=12, swing2LimitDegrees=12, twistLimitDegrees=12}
 {bone1Name=dyn_bangs_02, bone2Name=dyn_bangs_01, swing1LimitDegrees=12, swing2LimitDegrees=12, twistLimitDegrees=12}
 {bone1Name=dyn_bangs_01, bone2Name=dyn_bangs_00, swing1LimitDegrees=20, swing2LimitDegrees=12, twistLimitDegrees=12}
]
